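; Deprecated prototype cactus fruit actor that bursts to apply a healing effect nearby, then regrows.
; Vehicle overlap or damage (which also refills its health) sets the RepNotify flag WaitingToRegrow, whose handler calls burst or regrow.
; Bursting plays a GameplayCue, applies the healing GE through a temporary sphere overlap, and starts a timer that clears the flag.
; Burst and regrow visuals are driven by custom primitive data holding direction and start time.
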(event EventBeginPlay
  ;[1 1 1] Setup Refs - this is mostly for proto fruis that are growing through logic. When we have VFX/TA stuff this will likely be removed
  (Variables|Default|SetCactusLocation (Transformation|GetWorldLocation (Variables|Debug|GetStaticMeshComponent))))

(event OnComponentBeginOverlap(StaticMeshComponent)
  (params (OverlappedComponent :type Primitive%20Component%20Object%20Reference) (OtherActor :type Actor%20Object%20Reference) (OtherComp :type Primitive%20Component%20Object%20Reference) (OtherBodyIndex :type Integer) (bFromSweep :type Boolean) (SweepResult :type Hit%20Result%20Structure%20%28by%20ref%29))
  ;[1 1 1] Needed to trigger growth + burst from vehicle hit
  (bind _asfort_vehicle_interface (Utilities|Casting|CastToFortVehicleInterface OtherActor))
  (Variables|Default|SetInstigatingPawn (Vehicle|GetDriver _asfort_vehicle_interface))
  (if (not (Variables|Default|GetWaitingToRegrow))
    (Variables|Default|SetWaitingToRegrow true)))

(event EventOnDamageServer
  (params (Damage :type Float%20%28single-precision%29) (DamageTags :type Gameplay%20Tag%20Container%20Structure%20%28by%20ref%29) (Momentum :type Vector :default "0, 0, 0") (HitInfo :type Hit%20Result%20Structure%20%28by%20ref%29) (InstigatedBy :type Controller%20Object%20Reference) (DamageCauser :type Actor%20Object%20Reference) (EffectContext :type Gameplay%20Effect%20Context%20Handle%20Structure))
  ;[1 1 1] When hit, triggers fruit growth + burst
  (Health|SetHealth (Health|GetMaxHealth))
  (if (not (Variables|Default|GetWaitingToRegrow))
    (Variables|Default|SetWaitingToRegrow true)))

(event Custom|BurstPlants
  (bind _staticmeshcomponent (Variables|Debug|GetStaticMeshComponent))
  ;[1 1 1] Grows plants through timeline and pops them, playing GE and calling Heal function - mostly placeholder until we get VFX work
  (Rendering|Material|SetCustomPrimitiveDataFloat _staticmeshcomponent 1 -1.0)
  (Rendering|Material|SetCustomPrimitiveDataFloat _staticmeshcomponent 0 (Utilities|Time|GetGameTimeinSeconds))
  (GameplayCue|ExecuteGameplayCue self (Variables|Default|GetFruitBurstCue) (Utilities|Struct|MakeGameplayEffectContextHandle))
  (CallFunction|ApplyHeal)
  (Utilities|Time|SetTimerbyEvent (EventDispatchers|CreateEvent :!event Custom|TryRegrow) 1e+01))

(event Custom|ApplyHeal
  ;[1 0 0] Applying Heal through temporary sphere overlap -- NEEDS UPDATE
  (bind _outactors (Collision|SphereOverlapActors (Variables|Default|GetCactusLocation) 5e+02 (Utilities|Array|MakeArray "ObjectTypeQuery1" "ObjectTypeQuery2" "ObjectTypeQuery3" "ObjectTypeQuery6") 0 (Utilities|Array|MakeArray self)))
  (for _array_element _outactors
    (Fortnite|ApplyGameplayEffectsToActor _array_element (Utilities|Array|MakeArray (Variables|Default|GetHealingGE)) (Variables|Default|GetCactusLocation))))

(event Custom|TryRegrow
  ;[1 1 1] Regrowth events
  (Variables|Default|SetWaitingToRegrow false))

(event Custom|RegrowPlants
  (bind _staticmeshcomponent (Variables|Debug|GetStaticMeshComponent))
  ;[1 1 1] Regrowth events
  (Rendering|Material|SetCustomPrimitiveDataFloat _staticmeshcomponent 1 1.0)
  (Rendering|Material|SetCustomPrimitiveDataFloat _staticmeshcomponent 0 (Utilities|Time|GetGameTimeinSeconds)))

(fn OnRep_WaitingToRegrow
  (if (Variables|Default|GetWaitingToRegrow)
    (CallFunction|BurstPlants)
    (else
      (CallFunction|RegrowPlants))))

(fn ConstructionScript)
